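; Job board station prop that remembers which players have used it and shows them a switched-off "used" look.
; TrackServiceUser runs on the authority, adding the player's net ID to TrackedServiceUsers and forcing a net update for new entries.
; PlayServiceUsedEffects hides the has-jobs mesh and dims the no-jobs mesh's emissive via a dynamic material.
; Client BeginPlay replays those effects for a tracked local player, covering world-partition actor recreation.
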(event EventBeginPlay
  (bind _returnvalue (Fortnite|GetFirstLocalFortPlayerController true))
  (bind _playerstate (Variables|/Script/Engine.Controller|GetPlayerState _returnvalue))
  (bind _returnvalue_2 (PlayerState|GetUniqueNetId _playerstate))
  (|CallParentFunction)
  (Networking|SwitchHasAuthority
    (:Remote
      ;[1 1 1] FORT-610244 - world partitioning may tear down and recreate this actor client side. When beginning play, check if local user has used the job board, and if so, restore 'used' visuals
      (Utilities|IsValid _returnvalue
        (:Is%20Valid
          (Utilities|IsValid _playerstate
            (:Is%20Valid
              (bind _returnvalue_3 (UniqueNetID|IsValidNetID _returnvalue_2))
              (if _returnvalue_3
                (if (Utilities|Array|ContainsItem (Variables|Default|GetTrackedServiceUsers) _returnvalue_2)
                  (CallFunction|PlayServiceUsedEffects _returnvalue))))))))))

(event Collision|EventActorBeginOverlap
  (params (OtherActor :type Actor%20Object%20Reference))
  (|CallParentFunction OtherActor) ; This node is disabled and will not be called.\nDrag off pins to build functionality.
)

(event EventTick
  (params (DeltaSeconds :type Float%20%28single-precision%29))
  (|CallParentFunction DeltaSeconds) ; This node is disabled and will not be called.\nDrag off pins to build functionality.
)

(fn PlayServiceUsedEffects
  (params (PlayerController :type Fort%20Player%20Controller%20Object%20Reference))
  (if (Rendering|IsVisible (Variables|Default|GetMesh_HasJobs))
    (Rendering|SetVisibility (Variables|Default|GetMesh_HasJobs) false false)
    ;[1 1 1] Create dynamic mat so we can modify light emissive when job board used to 'switch it off'
    ;
    (bind _returnvalue (Rendering|Material|CreateDynamicMaterialInstance (Variables|Default|GetMesh_NoJobs) 0 0 "None" :!class /Script/Engine.PrimitiveComponent))
    (Utilities|IsValid _returnvalue
      (:Is%20Valid
        (Rendering|Material|SetVectorParameterValue :self _returnvalue :ParameterName "SFX_Emissive_Tint(Expand)" :!class /Script/Engine.MaterialInstanceDynamic)))))

(fn TrackServiceUser
  (params (PlayerController :type Fort%20Player%20Controller%20Object%20Reference))
  (bind _playerstate (Variables|/Script/Engine.Controller|GetPlayerState PlayerController))
  (bind _returnvalue (PlayerState|GetUniqueNetId _playerstate))
  (Networking|SwitchHasAuthority
    (:Authority
      (Utilities|IsValid PlayerController
        (:Is%20Valid
          (Utilities|IsValid _playerstate
            (:Is%20Valid
              (bind _returnvalue_2 (UniqueNetID|IsValidNetID _returnvalue))
              (if _returnvalue_2
                (bind _returnvalue_3 (Utilities|Array|AddUnique (Variables|Default|GetTrackedServiceUsers) _returnvalue))
                (if (> _returnvalue_3 -1)
                  (Networking|ForceNetUpdate))))))))))

(fn ConstructionScript
  (|CallParentFunction))
