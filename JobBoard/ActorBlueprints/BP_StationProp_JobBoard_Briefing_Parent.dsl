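; Parent Blueprint for the Job Board briefing station, a conversation prop that becomes non-interactable once used.
; Deactivation sets replicated bAlreadyInteractedWith, whose OnRep disables conversation, interaction and minimap marker, swaps to the closed-lid mesh and plays effects.
; An optional server check on safe zone phase changes deactivates boards caught in the storm.
; Multicasts handle conversation audio, and client BeginPlay restores used visuals when world partition or relevancy recreates the actor.
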(event EventEndPlay
  (params (EndPlayReason :type EEndPlayReason%20Enum :default "Destroyed"))
  (bind _returnvalue (SafeZone|GetSafeZoneIndicator))
  (|CallParentFunction EndPlayReason)
  (Utilities|IsValid _returnvalue
    (:Is%20Valid
      (EventDispatchers|UnbindEventfromOnSafeZonePhaseChanged _returnvalue (EventDispatchers|CreateEvent :!event Custom|OnSafeZoneChanged)))))

(event Custom|HandleStormChangeBehavior
  ;[0 0 0] Execute the following if actor is set to check state on storm phase
  (Networking|SwitchHasAuthority
    (:Authority
      (if (not (Variables|HandleStormChangeBehavior|GetbCheckedState))
        (Variables|HandleStormChangeBehavior|SetbCheckedState true) ; only call once for this actor
        (Class|FortAsyncActionCurrentPlaylistReady|PlaylistDataReadyAsync
          (:OnReady
            (bind _returnvalue (GameplayEventRouter|GetEventRouter "/Script/FortniteGame.FortGameState" self))
            (bind (_asynctaskproxy _payload) (GameplayEventRouter|StartListeningToStatefulEvent _returnvalue "/Script/FortniteGame.FortSafeZoneIndicatorUpdatedEvent")
              (:then
                ;;[1 1 1] NOTE: This may get set after having run OnGamePhaseChanged (if SavedEventStateExists runs)
                (Variables|Default|SetGamePhaseUpdatedAsyncTask _asynctaskproxy))
              (:OnEventReceived
                _)
              (:OnSavedEventStateExists
                _))
            (EventDispatchers|BindEventtoOnSafeZonePhaseChanged (SafeZone|GetSafeZoneIndicator) (EventDispatchers|CreateEvent :!event Custom|OnSafeZoneChanged))))))))

(event Custom|OnSafeZoneChanged
  (bind _returnvalue (BattleRoyaleGamePhaseLogic|GetBattleRoyaleGamePhaseLogicComponent))
  ;[0 0 0] Execute the following if actor is set to check state on storm phase
  ;;[0.1302 0.1302 0.1302] Check if the actor is in the storm when it changes; if so, deactivate it
  (Utilities|IsValid _returnvalue
    (:Is%20Valid
      (if (not (BattleRoyaleGamePhaseLogic|IsinCurrentSafeZone _returnvalue (Transformation|GetActorLocation) false false))
        (Development|PrintString "NOT in safe zone - deactivating " true true "(R=1.000000,G=0.000000,B=0.433296,A=1.000000)" 6.0)
        (Networking|SwitchHasAuthority
          (:Authority
            (CallFunction|DeactivateonUse)))))))

(event EventBeginPlay
  (bind _returnvalue (Fortnite|GetFirstLocalFortPlayerController true))
  ;[0 0 0] Clientside - when the actor is torn down from being out of relevancy, replay used effects
  (|CallParentFunction)
  (if (Variables|Default|GetbAlreadyInteractedWith)
    (Networking|SwitchHasAuthority
      (:Remote
        ;[1 1 1] FORT-610244 - world partitioning may tear down and recreate this actor client side. When beginning play, restore 'used' visuals if this machine is deactivated
        (Utilities|IsValid _returnvalue
          (:Is%20Valid
            (CallFunction|PlayServiceUsedEffects _returnvalue)))))
    (elif (Variables|HandleStormChangeBehavior|GetbCheckStateOnStormPhaseChange)
      (CallFunction|HandleStormChangeBehavior))))

(event Custom|Interaction%20Destroy
  ;[1 1 1] Server Trigger for Dynamic Destruction and Death Event
  (Utilities|Time|SetTimerbyEvent (delegate Custom|TriggerDie) (Variables|Default|GetLifespan%20Delay)))

(event Custom|TriggerDie
  ;[1 1 1] Server Trigger for Dynamic Destruction and Death Event
  (Gameplay|Die))

(event Custom|DeactivateOnUse
  ;[0 0 0] When side effect DeactivateOnUse is called, place actor in a no-interact state
  (Networking|SwitchHasAuthority
    (:Authority
      (Variables|Default|SetbAlreadyInteractedWith true))))

(event OnConversationEnded(NonPlayerConversationComponent)
  ;[0 0 0] Handles conversation audio
  (CallFunction|MultiCastonConversationEnded))

(event Custom|MultiCast_OnConversationEnded
  ;[0 0 0] Handles conversation audio
  (Networking|CanExecuteCosmeticEvents
    (:True
      (Audio|SpawnSoundAttached "/JobBoard/Sounds/JobBoard_ShadowBriefing_Cancel_Cue.JobBoard_ShadowBriefing_Cancel_Cue" (Variables|Debug|GetStaticMeshComponent))
      (bind _jobboardinteracting_audiocomponent (Variables|Default|GetJobBoardInteracting_AudioComponent)
        (:then
          (Audio|Components|Audio|Stop _jobboardinteracting_audiocomponent))
        (:else)))))

(event OnEnterConversationStateDelegate(NonPlayerConversationComponent)
  ;[0 0 0] Handles conversation audio
  (CallFunction|MultiCastonConversationEnterred))

(event Custom|MultiCast_OnConversationEnterred
  ;[0 0 0] Handles conversation audio
  (Networking|CanExecuteCosmeticEvents
    (:True
      (bind _returnvalue (Audio|SpawnSoundAttached :Sound "/JobBoard/Sounds/JobBoard_ShadowBriefing_Open_Cue.JobBoard_ShadowBriefing_Open_Cue" :AttachToComponent (Variables|Debug|GetStaticMeshComponent) :AttachPointName "None" :LocationType "KeepRelativeOffset" :bStopWhenAttachedToDestroyed true :VolumeMultiplier 1.0 :PitchMultiplier 1.0 :StartTime 0.0 :bAutoDestroy true))
      (Variables|Default|SetJobBoardInteracting_AudioComponent _returnvalue))))

(fn BlueprintCanInteract
  (params (InteractingPawn :type Fort%20Pawn%20Object%20Reference) (InteractionBeingAttempted :type EInteractionBeingAttempted%20Enum :default "FirstInteraction") (InteractionType :type TInteractionType%20Enum :default "IT_NoInteraction"))
  (return (|CallParentFunction InteractingPawn InteractionBeingAttempted InteractionType)))

(fn GetIconPlacement
  (params (SelfActor :type Actor%20Object%20Reference) (ViewingActor :type Actor%20Object%20Reference))
  (return (+ (Transformation|GetWorldLocation (Variables|Default|GetMesh_HasJobs)) (Variables|Interaction|GetIconPlacementOffset)) "0, 0, 0"))

(fn OnRep_bAlreadyInteractedWith
  (bind _returnvalue (Actor|GetComponentbyClass "/Script/FortniteGame.AthenaSpecialActorComponent"))
  (bind _nonplayerconversationcomponent (Variables|Components|GetNonPlayerConversationComponent))
  ;[0 0 0] Deactivate the terminal when this variable is set - can happen when interacted with or when it comes back into relevancy after being interacted with
  (if (Variables|Default|GetbAlreadyInteractedWith)
    ;;[0.1198 0.1198 0.1198] Destroy conversation component and disallow interaction
    (Variables|Interaction|SetbAllowInteract false)
    (InteractionUI|AddHideUIIndicatorReason _nonplayerconversationcomponent (Utilities|Name|MakeLiteralName "Deactivated since station entered a non-interactable state"))
    (Components|Activation|Deactivate _nonplayerconversationcomponent)
    ;;[0.151 0.151 0.151] Remove the minimap marker
    (Utilities|IsValid _returnvalue
      (:Is%20Valid
        (Components|Activation|Deactivate _returnvalue))
      (:Is%20Not%20Valid
        _))
    ;;[0.1094 0.1094 0.1094] Hide the mesh with open lid and unhide the mesh with closed lid
    (Utilities|IsValid (Variables|/Script/Engine.StaticMeshComponent|GetStaticMesh (Variables|Default|GetMesh_NoJobs))
      (:Is%20Valid
        (Development|SetHiddeninGame (Variables|Default|GetMesh_HasJobs) true true)
        (Development|SetHiddeninGame (Variables|Default|GetMesh_NoJobs) false true)
        ;;[0.1356 0.1356 0.1356] Disable collision from mesh and use simplified box collision for closed lid version.
        (Collision|SetCollisionEnabled (Variables|Default|GetMesh_HasJobs) "NoCollision")
        (Collision|SetCollisionEnabled (Variables|Default|GetCollision_NoJobs) "QueryAndPhysics")
        ;;[0.1354 0.1354 0.1354] Trigger cosmetic events for the lid closing
        (Networking|CanExecuteCosmeticEvents
          (:True
            (Rendering|Material|SetCustomPrimitiveDataFloat (Variables|Default|GetMesh_NoJobs) 0 (Utilities|Time|GetGameTimeinSeconds)) ; Start lid closing animation
          )))
      (:Is%20Not%20Valid
        (Networking|CanExecuteCosmeticEvents
          (:True
            _))))
    (Niagara|SpawnSystemAttached "/JobBoard/FX/NS_JobBoard_QuestFX_DustPoof.NS_JobBoard_QuestFX_DustPoof" (Variables|Default|GetMesh_HasJobs) "None" "0.000000,0.000000,110.000000")
    (Audio|PlaySoundatLocation (Variables|Effects|GetJobBoardDeactivatedSound) (Transformation|GetWorldLocation (Variables|Debug|GetStaticMeshComponent)))
    (bind _jobboardinteracting_audiocomponent (Variables|Default|GetJobBoardInteracting_AudioComponent)
      (:then
        (Audio|Components|Audio|Stop _jobboardinteracting_audiocomponent)
        (bind _station_ambient (Variables|Components|GetStation_Ambient)
          (:then
            (Audio|Components|Audio|FadeOut _station_ambient 1.0 0.0 "Linear" :!class /Script/Engine.AudioComponent))
          (:else)))
      (:else))))

(fn PlayServiceUsedEffects
  (params (PlayerController :type Fort%20Player%20Controller%20Object%20Reference))
  (bind _mesh_nojobs (Variables|Default|GetMesh_NoJobs))
  (if (Rendering|IsVisible (Variables|Default|GetMesh_HasJobs))
    ;[1 1 1] Create dynamic mat so we can modify light emissive when job board used to 'switch it off'
    (Utilities|IsValid (Variables|/Script/Engine.StaticMeshComponent|GetStaticMesh _mesh_nojobs)
      (:Is%20Valid
        (bind _returnvalue (Rendering|Material|CreateDynamicMaterialInstance _mesh_nojobs 0 0 "None" :!class /Script/Engine.PrimitiveComponent))
        (Utilities|IsValid _returnvalue
          (:Is%20Valid
            (Rendering|Material|SetVectorParameterValue :self _returnvalue :ParameterName "SFX_Emissive_Tint(Expand)" :!class /Script/Engine.MaterialInstanceDynamic)))))))

(fn TrackServiceUser
  (params (PlayerController :type Fort%20Player%20Controller%20Object%20Reference))
  (bind _playerstate (Variables|/Script/Engine.Controller|GetPlayerState PlayerController))
  (bind _returnvalue (PlayerState|GetUniqueNetId _playerstate))
  (Networking|SwitchHasAuthority
    (:Authority
      (Utilities|IsValid PlayerController
        (:Is%20Valid
          (Utilities|IsValid _playerstate
            (:Is%20Valid
              (bind _returnvalue_2 (UniqueNetID|IsValidNetID _returnvalue))
              (if _returnvalue_2
                (bind _returnvalue_3 (Utilities|Array|AddUnique (Variables|Default|GetTrackedServiceUsers) _returnvalue))
                (if (> _returnvalue_3 -1)
                  (Networking|ForceNetUpdate))))))))))

(fn ConstructionScript
  (|CallParentFunction)
  ;[0 0 0] FORT-791906 - Null check the static mesh on Mesh_NoJobs component to make sure it is set in children (and gracefully handles it if it isn't)
  (Utilities|FlowControl|Sequence
    (:then_0
      (Utilities|IsValid (Variables|/Script/Engine.StaticMeshComponent|GetStaticMesh (Variables|Default|GetMesh_NoJobs))
        (:Is%20Valid
          (Variables|Default|SetMeshArray (Utilities|Array|MakeArray (Variables|Default|GetMesh_HasJobs) (Variables|Default|GetMesh_NoJobs))))
        (:Is%20Not%20Valid
          (Variables|Default|SetMeshArray (Utilities|Array|MakeArray (Variables|Default|GetMesh_HasJobs))))))
    (:then_1
      (Effects|Components|ParticleSystem|SetColorParameter (Variables|Default|GetNiagara) "Color" (Variables|Visual|GetTintColor)))))
